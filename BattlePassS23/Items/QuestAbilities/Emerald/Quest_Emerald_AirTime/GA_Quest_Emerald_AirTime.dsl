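; Battle Pass quest ability that credits the player for seconds spent airborne after a shockwave grenade knock.
; Adding the ShockwaveGrenade_Knocked tag starts a looping one-second timer and removing it clears the timer; event-triggered activation also starts the timer directly.
; Each tick with MovementMode.Falling on the pawn adds a second, sends it to the quest manager as a complex custom stat event, then resets the counter.
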(event Ability|EventActivateAbilityFromEvent
  (params (EventData :type Gameplay%20Event%20Data%20Structure%20%28by%20ref%29))
  ;[1 1 1] Emerald Air Time
  (bind _returnvalue (Utilities|Time|SetTimerbyEvent (delegate Custom|Watch%20for%20Emerald%20AirTime) 1.0 true))
  (Variables|Default|SetTimer%20-%20AirTime _returnvalue))

(event Ability|EventActivateAbility
  (bind _shockwavegrenade_knocked (Variables|Default|GetShockwaveGrenade_Knocked))
  ;[1 1 1] Emerald Air Time
  (|CallParentFunction)
  (Utilities|FlowControl|Sequence
    (:then_0
      (Ability|Tasks|WaitGameplayTagAdd _shockwavegrenade_knocked
        (:Added
          (bind _returnvalue (Utilities|Time|SetTimerbyEvent (delegate Custom|Watch%20for%20Emerald%20AirTime) 1.0 true))
          (Variables|Default|SetTimer%20-%20AirTime _returnvalue))) ; added tag: ShockwaveGrenade_Knocked
    )
    (:then_1
      (Ability|Tasks|WaitGameplayTagRemove _shockwavegrenade_knocked
        (:Removed
          (Utilities|Time|ClearandInvalidateTimerbyHandle (Variables|Default|GetTimer%20-%20AirTime)))) ; removed tag: falling immunity
    )))

(event Custom|Watch%20for%20Emerald%20AirTime
  ;[1 1 1] Emerald Air Time
  (bind _bdidactorhavetags (Class|AthenaQuestFunctionLibrary|AthenaCheckActorHasTags (Variables|GAVars|Data|GetOwningPlayerPawn) "(GameplayTags=((TagName=\"MovementMode.Falling\")))")) ; In the Air
  (if _bdidactorhavetags
    (Math|Integer|IncrementInt (Variables|Default|Gettime%20%28seconds%29)
      (:%20%20
        (if (>= (Variables|Default|Gettime%20%28seconds%29) 1)
          (Quest|SendComplexCustomStatEvent (Variables|GAVars|Data|GetOwningPlayerQuestManager) 0 (Variables|Default|GetT_Quest) "(GameplayTags=)" (Variables|Default|Gettime%20%28seconds%29))
          (Development|PrintString (Utilities|String|Append "LogFortQuest [Shockwave Air Time] - sending Complex Custom, seconds accrued = " (Utilities|String|ToString(Integer) (Variables|Default|Gettime%20%28seconds%29))) false) ; debug - when quest event sent
          (Variables|Default|Settime%20%28seconds%29 0) ; reset accrued airtime counter so we only \nsend the amount accrued since the last\ntime we sent the event
        ))) ; +1 SECOND WILD AIRTIME
  ))

(event Ability|EventFailedToActivatePassiveAbility
  (params (ActorInfo :type Gameplay%20Ability%20Actor%20Info%20Structure)))
